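# Control Rig for the male medium Bandage Ninja body that adjusts three pelvis-parented pouch bones as the legs move.
# Spherical pose readers on each thigh output weights that blend offsets onto the pouches: left drives A and B, right drives C.
# The `crouching` variable, smoothed by AlphaInterp, additionally shifts and scales pouch B.
# A parent constraint from the right thigh to pouch C is present but not connected to execution.
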ControlRigBlueprint /BRCosmetics/Characters/Player/Male/Medium/Bodies/M_MED_Bandage_Ninja_01/Meshes/M_MED_Bandage_Ninja_Body_CtrlRig
graphs [
 RigVMFunctionLibrary {links [], nodes []}
 RigVMModel {nodes [
   RigUnit_BeginExecution = &RigUnit_BeginExecution()
   n1 = &RigUnit_SphericalPoseReader(
    ActiveRegionScaleFactors=(PositiveWidth=3.5,NegativeWidth=0.5,PositiveHeight=3,NegativeHeight=3)
    ActiveRegionSize=0.1
    Debug=(bDrawDebug=false,DebugScale=25,DebugSegments=20,DebugThickness=0.25)
    DriverAxis=(X=-1,Y=0,Z=0)
    DriverItem=(Type=Bone,Name="thigh_l")
    FalloffRegionScaleFactors=(PositiveWidth=1,NegativeWidth=1,PositiveHeight=1,NegativeHeight=1)
    FalloffSize=0.45
    OptionalParentItem=(Type=Bone)
    RotationOffset=(X=90,Y=0,Z=0)
    ExecutePin=@RigUnit_BeginExecution.ExecutePin
   )
   > &RigUnit_ModifyTransforms(
    ItemToModify=((Item=(Type=Bone,Name="dyn_pouch_a"),Transform=(Rotation=(X=0,Y=0,Z=0,W=1),Translation=(X=0,Y=-1,Z=0),Scale3D=(X=1,Y=1,Z=1))),(Item=(Type=Bone,Name="dyn_pouch_b"),Transform=(Rotation=(X=0,Y=0,Z=0,W=1),Translation=(X=0,Y=0,Z=0),Scale3D=(X=1,Y=1,Z=1))))
    ItemToModify.0.Transform.Rotation=&RigVMFunction_MathQuaternionFromEuler(Euler=(X=0,Y=0,Z=-15)).Result
    ItemToModify.1.Transform.Rotation=&RigVMFunction_MathQuaternionFromEuler(Euler=(X=0,Y=-90,Z=0)).Result
    Weight=@n1.OutputParam
   )
   > n2 = &RigUnit_SphericalPoseReader(
    ActiveRegionScaleFactors=(PositiveWidth=3.5,NegativeWidth=0.5,PositiveHeight=3,NegativeHeight=3)
    ActiveRegionSize=0.1
    Debug=(bDrawDebug=false,DebugScale=25,DebugSegments=20,DebugThickness=0.25)
    DriverAxis=(X=1,Y=0,Z=0)
    DriverItem=(Type=Bone,Name="thigh_r")
    FalloffRegionScaleFactors=(PositiveWidth=1,NegativeWidth=1,PositiveHeight=1,NegativeHeight=1)
    FalloffSize=0.45
    FlipHeightScaling=false
    FlipWidthScaling=true
    OptionalParentItem=(Type=Bone)
    RotationOffset=(X=-90,Y=0,Z=0)
   )
   > &RigUnit_ModifyTransforms(
    ItemToModify=((Item=(Type=Bone,Name="dyn_pouch_c"),Transform=(Rotation=(X=0,Y=0,Z=0,W=1),Translation=(X=0,Y=1,Z=0),Scale3D=(X=1,Y=1,Z=1))))
    ItemToModify.0.Transform.Rotation=&RigVMFunction_MathQuaternionFromEuler(Euler=(X=0,Y=0,Z=30)).Result
    Weight=@n2.OutputParam
   )
   AlphaInterp = &RigVMFunction_AlphaInterp(InterpSpeedDecreasing=10, InterpSpeedIncreasing=10, Scale=1, bClampResult=true, Value=RigVMVariableNode() {variable=crouching}.Value)
   > &RigUnit_ModifyTransforms(
    ItemToModify=((Item=(Type=Bone,Name="dyn_pouch_b"),Transform=(Rotation=(X=0,Y=0,Z=0,W=1),Translation=(X=-5,Y=0,Z=-2),Scale3D=(X=0.2,Y=0.2,Z=0.2))))
    Weight=@AlphaInterp.Result
   )
   &RigUnit_ParentConstraint(
    AdvancedSettings=(RotationOrderForFilter=XZY)
    Child=(Type=Bone,Name="dyn_pouch_c")
    ExecutePin=()
    Filter=(TranslationFilter=(bX=false,bY=false,bZ=false),RotationFilter=(bX=false,bY=True,bZ=false),ScaleFilter=(bX=True,bY=True,bZ=True))
    Parents=((Item=(Type=Bone,Name="thigh_r"),Weight=1))
    Weight=1
   )
  ]}
]
hierarchy [
 root {loc=<0 0 0>, rot=<0 0 0 1>}
 pelvis {parent=root, loc=<0 -0.7071 93.1191>, rot=<0.0127 0.707 -0.0127 -0.707>}
 spine_01 {parent=pelvis, loc=<6.7494 -0.6388 0>, rot=<0 0 0.034 -0.9994>}
 spine_02 {parent=spine_01, loc=<7.2518 -0.035 0>, rot=<0 0 0.0238 -0.9997>}
 spine_03 {parent=spine_02, loc=<6.0104 0 0>, rot=<0 0 0.0149 -0.9999>}
 spine_04 {parent=spine_03, loc=<8.1024 0 0>, rot=<0 0 -0.0132 -0.9999>}
 spine_05 {parent=spine_04, loc=<10.0146 0 0>, rot=<0 0 -0.0015 -1>}
 clavicle_l {parent=spine_05, loc=<4.3575 -0.1466 -2.3223>, rot=<0.0465 -0.7289 -0.009 -0.6829>}
 upperarm_l {parent=clavicle_l, loc=<15.5111 0 0>, rot=<0.0305 -0.536 -0.0057 -0.8437>}
 lowerarm_l {parent=upperarm_l, loc=<26.6514 0 0>, rot=<0.0391 -0.0099 -0.2453 -0.9686>}
 hand_l {parent=lowerarm_l, loc=<22.3491 0 0>, rot=<-0.7081 -0.0507 0.0975 -0.6975>}
 index_metacarpal_l {parent=hand_l, loc=<1.6535 0.0183 -2.96>, rot=<0.0158 -0.0124 -0.0173 -0.9996>}
 index_01_l {parent=index_metacarpal_l, loc=<8.3008 0.0357 -0.0524>, rot=<-0.0044 0.0485 0.0924 -0.9945>}
 index_02_l {parent=index_01_l, loc=<3.3587 -0.3004 0.0452>, rot=<-0.002 0.0027 0.3225 -0.9466>}
 index_03_l {parent=index_02_l, loc=<2.5578 -0.123 -0.0252>, rot=<-0.002 0.0067 0.2558 -0.9667>}
 middle_metacarpal_l {parent=hand_l, loc=<2.1856 -0.1725 -0.5984>, rot=<0.006 0.0168 -0.0329 -0.9993>}
 middle_01_l {parent=middle_metacarpal_l, loc=<7.3275 -0.2447 0.0129>, rot=<-0.0208 0.0042 0.0858 -0.9961>}
 middle_02_l {parent=middle_01_l, loc=<4.0123 0.1401 0.1175>, rot=<-0.0218 0.0258 0.3512 -0.9357>}
 middle_03_l {parent=middle_02_l, loc=<2.6077 0.0534 0.1373>, rot=<0.0082 0.0188 0.1708 -0.9851>}
 pinky_metacarpal_l {parent=hand_l, loc=<1.7616 -0.9753 3.3814>, rot=<-0.0397 0.0814 -0.069 -0.9935>}
 pinky_01_l {parent=pinky_metacarpal_l, loc=<7.1709 -0.4831 -0.1094>, rot=<-0.0654 -0.0035 0.2242 -0.9723>}
 pinky_02_l {parent=pinky_01_l, loc=<2.8274 0.0893 -0.1043>, rot=<0.0037 0.0011 0.2854 -0.9584>}
 pinky_03_l {parent=pinky_02_l, loc=<2.2046 0.0307 -0.0078>, rot=<-0.0002 0.0098 0.1445 -0.9895>}
 ring_metacarpal_l {parent=hand_l, loc=<2.2074 -0.3625 1.3638>, rot=<-0.0284 0.0722 -0.0458 -0.9959>}
 ring_01_l {parent=ring_metacarpal_l, loc=<7.0295 -0.6197 -0.1446>, rot=<-0.0489 -0.0252 0.1892 -0.9804>}
 ring_02_l {parent=ring_01_l, loc=<3.6174 0.4647 -0.1027>, rot=<0 0 0.2806 -0.9598>}
 ring_03_l {parent=ring_02_l, loc=<2.4152 0.114 0.0199>, rot=<0 0 0.1189 -0.9929>}
 thumb_01_l {parent=hand_l, loc=<1.3118 -0.3191 -2.459>, rot=<0.6546 -0.4042 0.0664 -0.6354>}
 thumb_02_l {parent=thumb_01_l, loc=<5.9336 -0.0448 0.0401>, rot=<-0.0037 0.0213 0.298 -0.9543>}
 thumb_03_l {parent=thumb_02_l, loc=<2.2223 0.3682 -0.0907>, rot=<0.0118 -0.013 0.0066 -0.9998>}
 lowerarm_twist_01_l {parent=lowerarm_l, loc=<17.8549 0 0>, rot=<0 0 0 1>}
 dyn_bandage_l_01 {parent=lowerarm_twist_01_l, loc=<-2.5322 -4.4257 -0.4175>, rot=<-0.6122 0.2292 0.4921 -0.5749>}
 dyn_bandage_l_02 {parent=dyn_bandage_l_01, loc=<1.171 -0.011 0.0233>, rot=<0 -0.1736 0 -0.9848>}
 dyn_bandage_l_03 {parent=dyn_bandage_l_02, loc=<1.7743 -0.0217 -0.022>, rot=<0 -0.1521 0 -0.9884>}
 dyn_bandage_l_04 {parent=dyn_bandage_l_03, loc=<5.0359 -0.0575 0.3723>, rot=<0 0 0 -1>}
 dyn_bandage_l_05 {parent=dyn_bandage_l_04, loc=<9.1416 -0.1346 0.0407>, rot=<0 0 0 -1>}
 dyn_bandage_l_06 {parent=dyn_bandage_l_05, loc=<13.106 -0.2102 0.0183>, rot=<0 0 0 -1>}
 dyn_bandage_l_07 {parent=dyn_bandage_l_06, loc=<14.8804 -0.251 0.0146>, rot=<0 0 0 -1>}
 lowerarm_twist_02_l {parent=lowerarm_l, loc=<9.6199 0 0>, rot=<0 0 0 1>}
 upperarm_twist_01_l {parent=upperarm_l, loc=<0.0041 0 0>, rot=<0 0 0 1>}
 upperarm_twist_02_l {parent=upperarm_l, loc=<13.7941 0 0>, rot=<0 0 0 1>}
 clavicle_r {parent=spine_05, loc=<4.3575 -0.1466 2.3223>, rot=<0.7289 0.0465 -0.6829 0.009>}
 upperarm_r {parent=clavicle_r, loc=<-15.5111 0 0>, rot=<0.0305 -0.536 -0.0057 -0.8437>}
 lowerarm_r {parent=upperarm_r, loc=<-26.6515 0 0>, rot=<0.0391 -0.0099 -0.2453 -0.9686>}
 hand_r {parent=lowerarm_r, loc=<-22.3492 0 0>, rot=<-0.7081 -0.0507 0.0975 -0.6975>}
 index_metacarpal_r {parent=hand_r, loc=<-1.6534 -0.0181 2.9601>, rot=<0.0158 -0.0124 -0.0173 -0.9996>}
 index_01_r {parent=index_metacarpal_r, loc=<-8.3008 -0.0357 0.0524>, rot=<-0.0044 0.0485 0.0924 -0.9945>}
 index_02_r {parent=index_01_r, loc=<-3.3589 0.3003 -0.0453>, rot=<-0.002 0.0027 0.3225 -0.9466>}
 index_03_r {parent=index_02_r, loc=<-2.5578 0.1231 0.0251>, rot=<-0.002 0.0067 0.2558 -0.9667>}
 middle_metacarpal_r {parent=hand_r, loc=<-2.1856 0.1726 0.5984>, rot=<0.006 0.0168 -0.0329 -0.9993>}
 middle_01_r {parent=middle_metacarpal_r, loc=<-7.3273 0.2447 -0.0129>, rot=<-0.0208 0.0042 0.0858 -0.9961>}
 middle_02_r {parent=middle_01_r, loc=<-4.0142 -0.1411 -0.07>, rot=<-0.0217 0.0259 0.3512 -0.9357>}
 middle_03_r {parent=middle_02_r, loc=<-2.6074 -0.0534 -0.1312>, rot=<0.0083 0.0188 0.1708 -0.9851>}
 pinky_metacarpal_r {parent=hand_r, loc=<-1.7617 0.9753 -3.3814>, rot=<-0.0397 0.0814 -0.069 -0.9935>}
 pinky_01_r {parent=pinky_metacarpal_r, loc=<-7.1707 0.4832 0.1094>, rot=<-0.0654 -0.0035 0.2242 -0.9723>}
 pinky_02_r {parent=pinky_01_r, loc=<-2.8274 -0.0892 0.1043>, rot=<0.0037 0.0011 0.2854 -0.9584>}
 pinky_03_r {parent=pinky_02_r, loc=<-2.2045 -0.031 0.0078>, rot=<-0.0002 0.0098 0.1445 -0.9895>}
 ring_metacarpal_r {parent=hand_r, loc=<-2.2073 0.3627 -1.3638>, rot=<-0.0284 0.0722 -0.0458 -0.9959>}
 ring_01_r {parent=ring_metacarpal_r, loc=<-7.0297 0.6197 0.1446>, rot=<-0.0489 -0.0252 0.1892 -0.9804>}
 ring_02_r {parent=ring_01_r, loc=<-3.6171 -0.4648 0.1027>, rot=<0 0 0.2806 -0.9598>}
 ring_03_r {parent=ring_02_r, loc=<-2.4153 -0.1141 -0.0199>, rot=<0 0 0.1189 -0.9929>}
 thumb_01_r {parent=hand_r, loc=<-1.3119 0.3192 2.459>, rot=<0.6546 -0.4042 0.0664 -0.6354>}
 thumb_02_r {parent=thumb_01_r, loc=<-5.9335 0.0447 -0.0401>, rot=<-0.0037 0.0213 0.298 -0.9543>}
 thumb_03_r {parent=thumb_02_r, loc=<-2.2223 -0.3682 0.0907>, rot=<0.0118 -0.013 0.0066 -0.9998>}
 lowerarm_twist_01_r {parent=lowerarm_r, loc=<-17.855 0 0>, rot=<0 0 0 1>}
 dyn_bandage_r_01 {parent=lowerarm_twist_01_r, loc=<2.5323 4.4258 0.4175>, rot=<0.4921 0.5749 0.6122 0.2292>}
 dyn_bandage_r_02 {parent=dyn_bandage_r_01, loc=<1.1709 0.011 0.0234>, rot=<0 -0.1736 0 -0.9848>}
 dyn_bandage_r_03 {parent=dyn_bandage_r_02, loc=<1.7744 0.0218 -0.0221>, rot=<0 -0.1521 0 -0.9884>}
 dyn_bandage_r_04 {parent=dyn_bandage_r_03, loc=<5.0358 0.0575 0.3723>, rot=<0 0 0 1>}
 dyn_bandage_r_05 {parent=dyn_bandage_r_04, loc=<9.1416 0.1346 0.0407>, rot=<0 0 0 1>}
 dyn_bandage_r_06 {parent=dyn_bandage_r_05, loc=<13.106 0.2103 0.0183>, rot=<0 -0.0282 0 -0.9996>}
 dyn_bandage_r_07 {parent=dyn_bandage_r_06, loc=<14.8804 0.2509 0.0146>, rot=<0 0 0 1>}
 lowerarm_twist_02_r {parent=lowerarm_r, loc=<-9.6199 0 0>, rot=<0 0 0 1>}
 upperarm_twist_01_r {parent=upperarm_r, loc=<-0.0041 0 0>, rot=<0 0 0 1>}
 upperarm_twist_02_r {parent=upperarm_r, loc=<-13.7942 0 0>, rot=<0 0 0 1>}
 neck_01 {parent=spine_05, loc=<11.3138 0.1529 0>, rot=<0 0 -0.2003 -0.9797>}
 neck_02 {parent=neck_01, loc=<4.5627 0 0>, rot=<0 0 0.0504 -0.9987>}
 head {parent=neck_02, loc=<5.3914 0 0>, rot=<0 0 0.115 -0.9934>}
 collar_r_mstr {parent=spine_05, loc=<2.5842 11.2818 6.2>, rot=<0.0284 -0.7065 -0.7065 0.0284>}
 collar_r_01 {parent=collar_r_mstr, loc=<0 0 0>, rot=<0 0 0 1>}
 collar_l_mstr {parent=spine_05, loc=<2.5841 11.2818 -6.2>, rot=<0.0284 -0.7065 -0.7065 0.0284>}
 collar_l_01 {parent=collar_l_mstr, loc=<0 0 0>, rot=<0 0 0 1>}
 dyn_coat_back {parent=spine_01, loc=<13.2491 -12.3434 -0.0155>, rot=<0.0114 -0.707 0.707 -0.0114>}
 dyn_coat_fl {parent=spine_01, loc=<10.762 6.7158 -12.4949>, rot=<0.0156 -0.9658 -0.2588 0.0042>}
 dyn_coat_bl {parent=spine_01, loc=<13.2169 -9.1209 -11.3178>, rot=<0.0143 -0.8869 0.4617 -0.0074>}
 dyn_coat_l {parent=spine_01, loc=<11.7666 -3.025 -13.8945>, rot=<0.0161 -0.9999 0 0>}
 dyn_coat_front_l {parent=spine_01, loc=<10.5063 10.5859 -8.5408>, rot=<0.014 -0.8659 -0.4999 0.0081>}
 dyn_coat_front_r {parent=spine_01, loc=<10.506 10.5859 8.5408>, rot=<-0.0081 0.4999 0.8659 -0.014>}
 dyn_coat_fr {parent=spine_01, loc=<10.762 6.7158 12.4949>, rot=<-0.0042 0.2588 0.9658 -0.0156>}
 dyn_coat_r {parent=spine_01, loc=<11.7666 -3.025 13.8945>, rot=<0 0 0.9999 -0.0161>}
 dyn_coat_br {parent=spine_01, loc=<13.2169 -9.1208 11.3178>, rot=<0.0074 -0.4617 0.8869 -0.0143>}
 thigh_l {parent=pelvis, loc=<-2.5693 -0.0109 -7.7501>, rot=<0.0012 0.0572 0.0155 -0.9982>}
 calf_l {parent=thigh_l, loc=<-39.0484 0 0>, rot=<0.0006 -0.0161 -0.0391 -0.9991>}
 calf_twist_01_l {parent=calf_l, loc=<-35.813 -0.0967 0.0397>, rot=<0.0237 0.0095 -0.0035 -0.9997>}
 calf_twist_02_l {parent=calf_l, loc=<-18.8064 -0.0967 0.0397>, rot=<0.0237 0.0095 -0.0035 -0.9997>}
 foot_l {parent=calf_l, loc=<-42.561 0.1077 0.4292>, rot=<0.0418 -0.0455 0.0343 -0.9975>}
 ball_l {parent=foot_l, loc=<-6.1599 13.8264 0.6806>, rot=<0.0001 -0.0001 -0.7063 -0.7079>}
 thigh_twist_01_l {parent=thigh_l, loc=<-21.6303 0 0>, rot=<0.0013 0 0 -1>}
 dyn_shorts_main_l {parent=thigh_twist_01_l, loc=<7.3423 -0.9031 0.2182>, rot=<-0.0027 -0.7464 0.0006 -0.6655>}
 dyn_shorts_out_l {parent=dyn_shorts_main_l, loc=<9.9469 -0.5145 1.2925>, rot=<0 0.1898 0 -0.9818>}
 dyn_shorts_out_02_l {parent=dyn_shorts_out_l, loc=<-2.685 0.2507 -13.1415>, rot=<0 0 0 1>}
 dyn_shorts_fr_l {parent=dyn_shorts_main_l, loc=<1.2885 10.5529 -1.7801>, rot=<-0.0067 0.0068 -0.7027 -0.7115>}
 dyn_shorts_fr_02_l {parent=dyn_shorts_fr_l, loc=<0 0 -12.0958>, rot=<0 0 0 1>}
 dyn_shorts_bk_l {parent=dyn_shorts_main_l, loc=<0.8878 -7.9198 0.31>, rot=<-0.0056 0.053 0.6996 -0.7125>}
 dyn_shorts_bk_02_l {parent=dyn_shorts_bk_l, loc=<-0.4147 0 -10.0565>, rot=<0 0 0 1>}
 dyn_shorts_in_l {parent=dyn_shorts_main_l, loc=<-8.6448 1.1393 -2.0369>, rot=<0.0069 0 1 0>}
 dyn_shorts_in_02_l {parent=dyn_shorts_in_l, loc=<-1.4897 0 -11.6935>, rot=<0 0 0 1>}
 thigh_r {parent=pelvis, loc=<-2.5691 -0.0109 7.7501>, rot=<0.0572 -0.0012 0.9982 0.0155>}
 calf_r {parent=thigh_r, loc=<39.0487 0 0>, rot=<0.0006 -0.0161 -0.0391 -0.9991>}
 calf_twist_01_r {parent=calf_r, loc=<35.8132 0.0967 -0.0397>, rot=<0.0237 0.0095 -0.0035 -0.9997>}
 calf_twist_02_r {parent=calf_r, loc=<18.8064 0.0967 -0.0397>, rot=<0.0237 0.0095 -0.0035 -0.9997>}
 foot_r {parent=calf_r, loc=<42.5612 -0.1077 -0.4292>, rot=<0.0418 -0.0455 0.0343 -0.9975>}
 ball_r {parent=foot_r, loc=<6.1598 -13.8264 -0.6806>, rot=<0.0001 -0.0001 -0.7063 -0.7079>}
 thigh_twist_01_r {parent=thigh_r, loc=<21.6303 0 0>, rot=<0.0013 0 0 -1>}
 dyn_shorts_main_r {parent=thigh_twist_01_r, loc=<-7.3421 0.9031 -0.2182>, rot=<-0.0027 -0.7464 0.0006 -0.6655>}
 dyn_shorts_out_r {parent=dyn_shorts_main_r, loc=<-9.9468 0.5145 -1.2924>, rot=<0 0.1898 0 -0.9818>}
 dyn_shorts_out_02_r {parent=dyn_shorts_out_r, loc=<2.685 -0.2507 13.1415>, rot=<0 0 0 1>}
 dyn_shorts_fr_r {parent=dyn_shorts_main_r, loc=<-1.2885 -10.5529 1.7801>, rot=<-0.0067 0.0068 -0.7027 -0.7115>}
 dyn_shorts_fr_02_r {parent=dyn_shorts_fr_r, loc=<0 0 12.0958>, rot=<0 0 0 1>}
 dyn_shorts_bk_r {parent=dyn_shorts_main_r, loc=<-0.8878 7.9198 -0.31>, rot=<-0.0056 0.053 0.6996 -0.7125>}
 dyn_shorts_bk_02_r {parent=dyn_shorts_bk_r, loc=<0.4147 0 10.0565>, rot=<0 0 0 1>}
 dyn_shorts_in_r {parent=dyn_shorts_main_r, loc=<8.6448 -1.1393 2.0369>, rot=<0.0069 0 1 0>}
 dyn_shorts_in_02_r {parent=dyn_shorts_in_r, loc=<1.4897 0 11.6936>, rot=<0 0 0 1>}
 dyn_pouch_l {parent=pelvis, loc=<0 0 0>, rot=<0.7071 -0.0013 -0.7071 -0.0013>}
 dyn_pouch_f {parent=pelvis, loc=<0 0 0>, rot=<0.7071 -0.0013 -0.7071 -0.0013>}
 dyn_pouch_r {parent=pelvis, loc=<0 0 0>, rot=<0.7071 -0.0013 -0.7071 -0.0013>}
 dyn_pouch_c {parent=pelvis, loc=<4.0254 1.247 15.5083>, rot=<0 0.0014 0.9998 0.0179>}
 dyn_pouch_a {parent=pelvis, loc=<3.9268 -0.4303 -15.7088>, rot=<-0.0179 -0.9998 0 0>}
 dyn_pouch_b {parent=pelvis, loc=<3.4964 8.8752 -10.8857>, rot=<-0.0163 -0.9107 -0.4127 -0.0074>}
]
variables [crouching {type=double, default=0}]
