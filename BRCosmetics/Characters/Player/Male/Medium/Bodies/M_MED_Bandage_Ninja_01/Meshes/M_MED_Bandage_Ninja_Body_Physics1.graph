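# Physics asset for the male medium Bandage Ninja body, set up for dynamic shorts bones around both legs.
# Capsules cover the thigh and calf twist bones and the calves, with no constraint linking thigh to calf.
# On each leg, four shorts chains (front, back, inner, outer) start as tiny capsules with free angular constraints to the thigh twist bone.
# Each chain ends in a flat box panel, tightly angle-limited to its root.
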PhysicsAsset /BRCosmetics/Characters/Player/Male/Medium/Bodies/M_MED_Bandage_Ninja_01/Meshes/M_MED_Bandage_Ninja_Body_Physics1
bodies {
 calf_l [{shapeType=Capsule, length=16.3435, radius=7.1565, center=<-14.2533 -3.2563 -0.4527>, rotation=<rot 90 -3.4277 176.0191>}]
 calf_r [{shapeType=Capsule, length=15.7407, radius=7.7593, center=<14.1599 2.6497 0.7585>, rotation=<rot 90 -10.024 -10.3989>}]
 calf_twist_02_l [{shapeType=Capsule, length=17.9618, radius=5.5382, center=<-20.1486 -0.2148 0.664>, rotation=<rot 86.3626 -10.2246 -179.9995>}]
 calf_twist_02_r [{shapeType=Capsule, length=18.25, radius=5.25, center=<20.2357 0.8088 -0.0002>, rotation=<rot 90 0 3.5324>}]
 dyn_shorts_bk_02_l [{shapeType=Box, extentX=4, extentY=18, extentZ=18, center=<2 0 -6.6448>}]
 dyn_shorts_bk_02_r [{shapeType=Box, extentX=4, extentY=18, extentZ=18, center=<-2 0 6.6448>}]
 dyn_shorts_bk_l [{shapeType=Capsule, length=1, radius=1}]
 dyn_shorts_bk_r [{shapeType=Capsule, length=1, radius=1}]
 dyn_shorts_fr_02_l [{shapeType=Box, extentX=3.5, extentY=18, extentZ=18, center=<2 0 -5>}]
 dyn_shorts_fr_02_r [{shapeType=Box, extentX=3.5, extentY=18, extentZ=18, center=<0 0 5>}]
 dyn_shorts_fr_l [{shapeType=Capsule, length=1, radius=1}]
 dyn_shorts_fr_r [{shapeType=Capsule, length=1, radius=1}]
 dyn_shorts_in_02_l [{shapeType=Box, extentX=3, extentY=18, extentZ=18, center=<1.5 0 -5>}]
 dyn_shorts_in_02_r [{shapeType=Box, extentX=3, extentY=18, extentZ=18, center=<-1.5 0 5>}]
 dyn_shorts_in_l [{shapeType=Capsule, length=1, radius=1}]
 dyn_shorts_in_r [{shapeType=Capsule, length=1, radius=1}]
 dyn_shorts_out_02_l [{shapeType=Box, extentX=6, extentY=18, extentZ=18, center=<0 0 -5>, rotation=<rot -15 0 0>}]
 dyn_shorts_out_02_r [{shapeType=Box, extentX=6.25, extentY=18, extentZ=18, center=<0 0 5>, rotation=<rot -15 0 -0.0001>}]
 dyn_shorts_out_l [{shapeType=Capsule, length=1, radius=1}]
 dyn_shorts_out_r [{shapeType=Capsule, length=1, radius=1}]
 thigh_twist_01_l [{shapeType=Capsule, length=26.25, radius=8.5, center=<-2.3357 0.5566 0>, rotation=<rot 90 0 -5>}]
 thigh_twist_01_r [{shapeType=Capsule, length=26.25, radius=8.5, center=<2.3357 0.5566 0>, rotation=<rot 90 0 -5>}]
}
constraints [
 {
  bone1Name=calf_twist_02_r
  bone2Name=calf_r
  swing1LimitDegrees=45
  swing1Motion=Free
  swing2LimitDegrees=45
  swing2Motion=Free
  twistLimitDegrees=45
  twistMotion=Free
 }
 {
  bone1Name=calf_twist_02_l
  bone2Name=calf_l
  swing1LimitDegrees=45
  swing1Motion=Free
  swing2LimitDegrees=45
  swing2Motion=Free
  twistLimitDegrees=45
  twistMotion=Free
 }
 {
  bone1Name=dyn_shorts_bk_r
  bone2Name=thigh_twist_01_r
  swing1LimitDegrees=45
  swing1Motion=Free
  swing2LimitDegrees=45
  swing2Motion=Free
  twistLimitDegrees=45
  twistMotion=Free
 }
 {bone1Name=dyn_shorts_bk_02_r, bone2Name=dyn_shorts_bk_r, swing1LimitDegrees=10, swing2LimitDegrees=10, twistLimitDegrees=2}
 {
  bone1Name=dyn_shorts_fr_r
  bone2Name=thigh_twist_01_r
  swing1LimitDegrees=45
  swing1Motion=Free
  swing2LimitDegrees=45
  swing2Motion=Free
  twistLimitDegrees=45
  twistMotion=Free
 }
 {bone1Name=dyn_shorts_fr_02_r, bone2Name=dyn_shorts_fr_r, swing1LimitDegrees=10, swing2LimitDegrees=10, twistLimitDegrees=2}
 {
  bone1Name=dyn_shorts_in_r
  bone2Name=thigh_twist_01_r
  swing1LimitDegrees=45
  swing1Motion=Free
  swing2LimitDegrees=45
  swing2Motion=Free
  twistLimitDegrees=45
  twistMotion=Free
 }
 {bone1Name=dyn_shorts_in_02_r, bone2Name=dyn_shorts_in_r, swing1LimitDegrees=10, swing2LimitDegrees=10, twistLimitDegrees=2}
 {
  bone1Name=dyn_shorts_out_r
  bone2Name=thigh_twist_01_r
  swing1LimitDegrees=45
  swing1Motion=Free
  swing2LimitDegrees=45
  swing2Motion=Free
  twistLimitDegrees=45
  twistMotion=Free
 }
 {bone1Name=dyn_shorts_out_02_r, bone2Name=dyn_shorts_out_r, swing1LimitDegrees=10, swing2LimitDegrees=10, twistLimitDegrees=2}
 {bone1Name=dyn_shorts_out_02_l, bone2Name=dyn_shorts_out_l, swing1LimitDegrees=10, swing2LimitDegrees=10, twistLimitDegrees=2}
 {
  bone1Name=dyn_shorts_bk_l
  bone2Name=thigh_twist_01_l
  swing1LimitDegrees=45
  swing1Motion=Free
  swing2LimitDegrees=45
  swing2Motion=Free
  twistLimitDegrees=45
  twistMotion=Free
 }
 {bone1Name=dyn_shorts_bk_02_l, bone2Name=dyn_shorts_bk_l, swing1LimitDegrees=10, swing2LimitDegrees=10, twistLimitDegrees=2}
 {
  bone1Name=dyn_shorts_fr_l
  bone2Name=thigh_twist_01_l
  swing1LimitDegrees=45
  swing1Motion=Free
  swing2LimitDegrees=45
  swing2Motion=Free
  twistLimitDegrees=45
  twistMotion=Free
 }
 {bone1Name=dyn_shorts_fr_02_l, bone2Name=dyn_shorts_fr_l, swing1LimitDegrees=10, swing2LimitDegrees=10, twistLimitDegrees=2}
 {
  bone1Name=dyn_shorts_in_l
  bone2Name=thigh_twist_01_l
  swing1LimitDegrees=45
  swing1Motion=Free
  swing2LimitDegrees=45
  swing2Motion=Free
  twistLimitDegrees=45
  twistMotion=Free
 }
 {bone1Name=dyn_shorts_in_02_l, bone2Name=dyn_shorts_in_l, swing1LimitDegrees=10, swing2LimitDegrees=10, twistLimitDegrees=2}
 {
  bone1Name=dyn_shorts_out_l
  bone2Name=thigh_twist_01_l
  swing1LimitDegrees=45
  swing1Motion=Free
  swing2LimitDegrees=45
  swing2Motion=Free
  twistLimitDegrees=45
  twistMotion=Free
 }
]
