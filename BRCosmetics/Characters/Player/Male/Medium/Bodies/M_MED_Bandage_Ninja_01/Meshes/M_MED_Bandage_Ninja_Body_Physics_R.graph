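# Physics asset for the right-arm dangling bandage on the male medium Bandage Ninja body, apparently for secondary cloth-like motion.
# Seven identical capsules form the dyn_bandage_r chain, constrained link to link from lowerarm_r with free swing.
# Coarse capsules on the pelvis, spine, clavicles, arms, thighs and calves have no constraints, so they likely serve as collision volumes for the chain.
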PhysicsAsset /BRCosmetics/Characters/Player/Male/Medium/Bodies/M_MED_Bandage_Ninja_01/Meshes/M_MED_Bandage_Ninja_Body_Physics_R
bodies {
 calf_l [{shapeType=Capsule, length=26.002, radius=10.5818, center=<-19.839 -0.0002 0>, rotation=<rot 90 0 0>}]
 calf_r [{shapeType=Capsule, length=25.6153, radius=10.4315, center=<19.0128 0 0>, rotation=<rot 90 0 0>}]
 clavicle_l [{shapeType=Capsule, length=10.3261, radius=8.5802, center=<8.2822 -0.2527 -2.0684>, rotation=<rot 90 -0.0049 179.9951>}]
 clavicle_r [{shapeType=Capsule, length=10.3261, radius=8.5802, center=<-7.9705 0.6201 2.6048>, rotation=<rot 90 0.0198 0.0198>}]
 dyn_bandage_r_01 [{shapeType=Capsule, length=6, radius=4, shapeName=dyn_bandage_r_01_capsule, center=<6 0 0>, rotation=<rot 90 0 0>}]
 dyn_bandage_r_02 [{shapeType=Capsule, length=6, radius=4, shapeName=dyn_bandage_r_02_capsule, center=<6 0 0>, rotation=<rot 90 0 0>}]
 dyn_bandage_r_03 [{shapeType=Capsule, length=6, radius=4, shapeName=dyn_bandage_r_03_capsule, center=<6 0 0>, rotation=<rot 90 0 0>}]
 dyn_bandage_r_04 [{shapeType=Capsule, length=6, radius=4, shapeName=dyn_bandage_r_04_capsule, center=<6 0 0>, rotation=<rot 90 0 0>}]
 dyn_bandage_r_05 [{shapeType=Capsule, length=6, radius=4, shapeName=dyn_bandage_r_05_capsule, center=<6 0 0>, rotation=<rot 90 0 0>}]
 dyn_bandage_r_06 [{shapeType=Capsule, length=6, radius=4, shapeName=dyn_bandage_r_06_capsule, center=<6 0 0>, rotation=<rot 90 0 0>}]
 dyn_bandage_r_07 [{shapeType=Capsule, length=6, radius=4, shapeName=dyn_bandage_r_07_capsule, center=<6 0 0>, rotation=<rot 90 0 0>}]
 lowerarm_l [{shapeType=Capsule, length=18.5, radius=4, center=<7.7891 0 0>, rotation=<rot -90 0.0198 -0.0199>}]
 lowerarm_r [{shapeType=Capsule, length=19, radius=4, center=<-10.9801 0.1538 -0.0002>, rotation=<rot 90 -0.0098 -0.0098>}]
 pelvis [{shapeType=Capsule, length=11.4559, radius=14.3329, center=<2.2747 -0.2766 1.0154>}]
 spine_03 [{shapeType=Capsule, length=10.5169, radius=16.5072, center=<-1.8305 2.0337 0>, rotation=<rot -83.7243 -83.7209 89.9955>}]
 spine_05 [{shapeType=Capsule, length=16.7232, radius=12.517, center=<-2.3724 -0.6607 0>}]
 thigh_l [{shapeType=Capsule, length=26.002, radius=10.5818, center=<-19.839 -0.0002 0>, rotation=<rot 90 0 0>}]
 thigh_r [{shapeType=Capsule, length=25.6153, radius=10.4315, center=<19.0128 0 0>, rotation=<rot 90 0 0>}]
 upperarm_l [{shapeType=Capsule, length=21.3261, radius=4, center=<12.1824 -0.2526 -0.6865>, rotation=<rot 90 0.005 -179.995>}]
 upperarm_r [{shapeType=Capsule, length=17.1357, radius=4, center=<-12.8316 -0.0074 -0.1319>, rotation=<rot 0 -88.4734 -86.4203>}]
}
constraints [
 {
  bone1Name=dyn_bandage_r_01
  bone2Name=lowerarm_r
  swing1LimitDegrees=45
  swing1Motion=Free
  swing2LimitDegrees=45
  swing2Motion=Free
  twistLimitDegrees=5
 }
 {
  bone1Name=dyn_bandage_r_02
  bone2Name=dyn_bandage_r_01
  swing1LimitDegrees=45
  swing1Motion=Free
  swing2LimitDegrees=45
  swing2Motion=Free
  twistLimitDegrees=5
 }
 {
  bone1Name=dyn_bandage_r_03
  bone2Name=dyn_bandage_r_02
  swing1LimitDegrees=45
  swing1Motion=Free
  swing2LimitDegrees=45
  swing2Motion=Free
  twistLimitDegrees=5
 }
 {
  bone1Name=dyn_bandage_r_04
  bone2Name=dyn_bandage_r_03
  swing1LimitDegrees=45
  swing1Motion=Free
  swing2LimitDegrees=45
  swing2Motion=Free
  twistLimitDegrees=5
 }
 {
  bone1Name=dyn_bandage_r_05
  bone2Name=dyn_bandage_r_04
  swing1LimitDegrees=45
  swing1Motion=Free
  swing2LimitDegrees=45
  swing2Motion=Free
  twistLimitDegrees=5
 }
 {
  bone1Name=dyn_bandage_r_06
  bone2Name=dyn_bandage_r_05
  swing1LimitDegrees=45
  swing1Motion=Free
  swing2LimitDegrees=45
  swing2Motion=Free
  twistLimitDegrees=5
 }
 {
  bone1Name=dyn_bandage_r_07
  bone2Name=dyn_bandage_r_06
  swing1LimitDegrees=45
  swing1Motion=Free
  swing2LimitDegrees=45
  swing2Motion=Free
  twistLimitDegrees=5
 }
]
