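# Physics asset for the face accessory mesh of the male medium Bandage Ninja player body cosmetic.
# It holds two capsule bodies, a short one on the dyn_hair bone and a zero-length, sphere-like one on the head bone.
# A single constraint ties dyn_hair to head with moderate swing limits and a tight twist limit, allowing the hair bone to sway with little twisting.
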PhysicsAsset /BRCosmetics/Characters/Player/Male/Medium/Bodies/M_MED_Bandage_Ninja_01/Meshes/M_MED_Bandage_Ninja_FaceAcc_Physics
bodies {
 dyn_hair [{shapeType=Capsule, length=5.5, radius=2.75, center=<3 0 0>, rotation=<rot 90 0 0>}]
 head [{shapeType=Capsule, length=0, radius=5}]
}
constraints [{bone1Name=dyn_hair, bone2Name=head, swing1LimitDegrees=45, swing2LimitDegrees=45, twistLimitDegrees=5}]
